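# Temporary parent material for FrigidPoppet's skinned character fur, built on the 2023 character PreFX/PostFX functions.
# A tech-art mask channel blends in subsurface-profile snow, with optional exposure-compensated emissive crevice boost.
# Switchable GlitterFX glints derive view-dependent sparkle from texture normals and sine noise, shaped by a sun-direction lobe, and can alter surface attributes.
# The result passes through fur parameter, shell and WPO-shrink functions into OUT_Material.
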Material /BRCosmetics/Characters/Player/Female/Medium/Bodies/F_MED_FrigidPoppet/Fur/M_FrigidPoppetFur_ParentTemp
graph {
 properties {
  bUseMaterialAttributes=true
  bUsedWithMeshDeformer=true
  bUsedWithMorphTargets=true
  bUsedWithSkeletalMesh=true
  pixelDepthOffsetMode=PDOM_Legacy
  shadingModel=MSM_FromMaterialExpression
  usedShadingModels=Default Lit | Subsurface | Clear Coat | Subsurface Profile | Cloth
 }
 nodes [
  n1 = MaterialFunctionCall() {materialFunction=&MF_CharacterParent_2023_PreFX}
  MaterialFunctionCall(
   Material=NamedRerouteUsage()
   M=@n1.M
   Pre Skinned Local Position=@n1.Pre skinned Local Position
   Distance Blend=@n1.Pre skinned Local Position
   Fresnel=@n1.Pre skinned Local Position
  ) {materialFunction=&MF_CharacterParent_2023_PostFX}
  n2 = GetMaterialAttributes(None=@n1.Result) {attributeGetTypes [#1, #2, #3]}
  n3 = LinearInterpolate(
   A=VectorParameter() {group=_AirKringle, parameterName=SnowSubsurfaceCreviceColorTint, defaultValue=<rgba 1 1 1 1>}.RGB
   B=VectorParameter() {group=_AirKringle, parameterName=SnowSubsurfaceFlatsColorTint, defaultValue=<rgba 1 1 1 1>}.RGB
   Alpha=Power(
    Base=ChannelMaskParameter(None=@n1.M) {group=_AirKringle, parameterName=SnowCreviceBoostChannel}.R
    Exp=ScalarParameter() {defaultValue=1, group=_AirKringle, parameterName=SnowCreviceBoost_MaskExp}
   )
  ) {constA=1}
  n4 = EyeAdaptationInverse(LightValueInput=@n3, AlphaInput=ScalarParameter() {group=_AirKringle, parameterName=SnowSubsurface_ExposureCompensationAmt, sliderMax=1})
  n5 = SetMaterialAttributes!(
   MaterialAttributes=@n2.MaterialAttributes
   Opacity=StaticSwitchParameter(True=@n2.Opacity, False=ScalarParameter() {defaultValue=1, group=_AirKringle, parameterName=SnowOpacity}) {group=_AirKringle, parameterName=UseFresnelSSSOpacity}
   Shading Model=ShadingModel() {shadingModel=MSM_SubsurfaceProfile}
   Fuzz Color or Subsurface Color=@n2.MaterialAttributes
   Emissive Color=StaticSwitchParameter(True=Add(A=@n2.Emissive Color, B=@n4.EyeAdaptationInverse), False=@n2.Emissive Color) {group=_AirKringle, parameterName=UseCreviceEmissiveBoost, sortPriority=-32}
  ) {attributeSetTypes [
    #1
    #4
    #2
    #3
   ]}
  n6 = GetMaterialAttributes(None=BlendMaterialAttributes(A=@n2.MaterialAttributes, B=@n5, Alpha=NamedRerouteUsage())) {attributeGetTypes [
    #5
    #6
    #7
    #3
    #8
   ]}
  n7 = NamedRerouteUsage()
  n8 = VectorParameter() {group=GlitterFX V2, parameterName=GlintsBaseColor, sortPriority=10, defaultValue=<rgba 1 1 1 1>}
  n9 = Multiply(A=NamedRerouteUsage(), B=ScalarParameter() {group=GlitterFX V2, parameterName=GlintsBaseColor_Influence, sortPriority=10})
  n10 = NamedRerouteUsage()
  n11 = ScalarParameter() {defaultValue=0.2, group=GlitterFX V2, parameterName=GlintsMetallic, sortPriority=15}
  n12 = Multiply(A=NamedRerouteUsage(), B=ScalarParameter() {group=GlitterFX V2, parameterName=GlintsMetallic_Influence, sortPriority=15})
  n13 = NamedRerouteUsage()
  n14 = ScalarParameter() {defaultValue=0.2, group=GlitterFX V2, parameterName=GlintsRoughness, sortPriority=20}
  n15 = Multiply(A=NamedRerouteUsage(), B=ScalarParameter() {defaultValue=0.2, group=GlitterFX V2, parameterName=GlintsRoughness_Influence, sortPriority=20})
  n16 = NamedRerouteUsage()
  n17 = LinearInterpolate(
   A=Constant3Vector() {constant=<rgba 0 0 1 0>}
   B=Transform(None=NamedRerouteUsage()) {transformSourceType=TRANSFORMSOURCE_World, transformType=TRANSFORM_Tangent}
   Alpha=Multiply(A=NamedRerouteUsage(), B=ScalarParameter() {defaultValue=1, group=GlitterFX V2, parameterName=GlintsNormal_Influence, sortPriority=25})
  )
  n18 = SetMaterialAttributes!(
   MaterialAttributes=@n6.MaterialAttributes
   Base Color=StaticSwitchParameter(True=LinearInterpolate(A=@n7, B=@n8.RGB, Alpha=@n9), False=@n7) {group=GlitterFX V2, parameterName=Glint_AffectBaseColor, sortPriority=9}
   Metallic=StaticSwitchParameter(True=LinearInterpolate(A=@n10, B=@n11, Alpha=@n12), False=@n10) {group=GlitterFX V2, parameterName=Glint_AffectMetallic, sortPriority=14}
   Roughness=StaticSwitchParameter(True=LinearInterpolate(A=@n13, B=@n14, Alpha=@n15), False=@n13) {group=GlitterFX V2, parameterName=Glint_AffectRoughness, sortPriority=19}
   Emissive Color=Add(A=NamedRerouteUsage(), B=NamedRerouteUsage())
   Normal=StaticSwitchParameter(
    True=MaterialFunctionCall(BaseNormal=@n16, AdditionalNormal=@n17) {materialFunction=&BlendAngleCorrectedNormals}.Result
    False=@n16
   ) {group=GlitterFX V2, parameterName=Glint_AffectNormal, sortPriority=24}
  ) {attributeSetTypes [
    #5
    #6
    #7
    #3
    #8
   ]}
  n19 = ChannelMaskParameter(None=@n1.M) {group=GlitterFX V2, parameterName=Glint Secondary Mask Channel}
  n20 = ScalarParameter() {defaultValue=2, group=GlitterFX V2, parameterName=Glint Secondary Mask Exponent}
  n21 = NamedRerouteDeclaration(None=LinearInterpolate(
    B=Power(Base=@n19.R, Exp=@n20)
    Alpha=ScalarParameter() {defaultValue=1, group=GlitterFX V2, parameterName=Glint Secondary Mask Influence, sliderMax=1}
   ) {constA=1}) {name=GlintSecondaryMask, nodeColor=<rgba 0.2891 1 0 1>}
  n22 = NamedRerouteUsage()
  n23 = StaticSwitchParameter(
   True=BlendMaterialAttributes(A=@n6.MaterialAttributes, B=@n18, Alpha=Multiply(A=@n21, B=@n22)) {vertexAttributeBlendType=UseA}
   False=@n6.MaterialAttributes
  ) {group=GlitterFX V2, parameterName=Enable Glints, sortPriority=-100}
  n24 = MaterialFunctionCall!(MA=@n23, TechArtMask=NamedRerouteUsage()) {materialFunction=/Game/Global/Materials/Character_Parent_2023/Fur_Parent_2023/MF_Fur_ParamAdjustments}
  MaterialFunctionCall!(In=@n24.Result, TechArtMask=NamedRerouteUsage(), Mask Input=NamedRerouteUsage()) {materialFunction=/Game/Global/Materials/Character_Parent_2023/Fur_Parent_2023/MF_Fur_Painted}
  MaterialFunctionCall!(Material=@n23, TechArtMask=NamedRerouteUsage()) {materialFunction=/Game/Global/Materials/Character_Parent_2023/Fur_Parent_2023/MF_Fur_CustomColor}
  Multiply(A=@n2.Fuzz Color or Subsurface Color, B=@n3)
  n25 = Multiply(
   A=Multiply(A=TextureCoordinate(), B=ScalarParameter() {defaultValue=1, group=GlitterFX V2, parameterName=Glint_Scale})
   B=ScalarParameter() {defaultValue=10, group=GlitterFX V2, parameterName=Glint_ScalePerMaterial}
  )
  n26 = TextureSampleParameter2D!(UVs=@n25, Bias=Constant() {r=-2}, Apply View MipBias) {
   group=GlitterFX V2
   mipValueMode=TMVM_MipBias
   parameterName=Glints_CoordsAndMask
   samplerType=SAMPLERTYPE_Masks
   texture=/BRCosmetics/Characters/Player/Female/Medium/Bodies/F_MED_AirKringle/Materials/T_Glints_M
  }
  n27 = TextureSampleParameter2D!(UVs=ComponentMask(None=@n26.RGB) {g=true, r=true}, "DDX(UVs)"=DDX(Value=@n25), "DDY(UVs)"=DDY(Value=@n25), Apply View MipBias) {
   group=GlitterFX V2
   mipValueMode=TMVM_Derivative
   parameterName=Glints_WorldDirection
   samplerSource=SSM_Wrap_WorldGroupSettings
   samplerType=SAMPLERTYPE_LinearColor
   texture=/BRCosmetics/Characters/Player/Female/Medium/Bodies/F_MED_AirKringle/Materials/T_Glints_N
  }
  n28 = ConstantBiasScale(None=@n27.RGB) {bias=-0.5, scale=2}
  n29 = Multiply(A=VertexNormalWS(), B=ScalarParameter() {group=GlitterFX V2, parameterName=GlintNormals_VertexNormalInfluence, sliderMax=1})
  n30 = NamedRerouteDeclaration(None=Normalize(VectorInput=Add(A=@n28, B=@n29))) {name=Glint Normal, nodeColor=<rgba 0 1 0.2734 1>}
  n31 = CameraVectorWS()
  n32 = Multiply(
   A=LinearInterpolate(
    A=Multiply(A=@n31) {constB=-1}
    B=MaterialFunctionCall() {materialFunction=&CameraDirectionVector}.Result
    Alpha=ScalarParameter() {group=GlitterFX V2, parameterName=GlintNormals_CameraVectorUniformity, sliderMax=1}
   )
   B=ScalarParameter() {defaultValue=2, group=GlitterFX V2, parameterName=GlintNormals_ScaleCameraVector}
  ) {constB=2}
  n33 = ScalarParameter() {defaultValue=15, group=GlitterFX V2, parameterName=Glint_NoiseFrequency}
  n34 = Frac(None=Multiply(A=DotProduct(A=@n30, B=@n32), B=@n33) {constB=7})
  n35 = Abs(None=Multiply(A=ConstantBiasScale(None=@n34) {bias=-0.5, scale=2}, B=Add(B=@n33) {constA=1}))
  n36 = LinearInterpolate(
   A=ScalarParameter() {defaultValue=0.25, group=GlitterFX V2, parameterName=Glint_NoiseThreshold, sliderMax=0.999}
   B=ScalarParameter() {defaultValue=0.25, group=GlitterFX V2, parameterName=Glint_NoiseThreshold_Lit, sliderMax=0.999}
   Alpha=NamedRerouteUsage()
  )
  n37 = Subtract(A=ConstantBiasScale(None=Sine(None=@n35)), B=@n36)
  n38 = OneMinus(None=@n36)
  n39 = Power(
   Base=Multiply(A=Divide(A=@n37, B=@n38), B=NamedRerouteUsage())
   Exp=ScalarParameter() {defaultValue=3, group=GlitterFX V2, parameterName=Glint_NoiseExponentBase}
  )
  n40 = SkyAtmosphereLightDirection()
  n41 = Add(A=@n40, B=CameraVectorWS())
  n42 = VertexNormalWS()
  n43 = DotProduct(A=@n42, B=@n40)
  n44 = Multiply(A=DotProduct(A=Normalize(VectorInput=@n41), B=@n42), B=Saturate(None=Add(A=@n43)))
  n45 = ScalarParameter() {defaultValue=1, group=GlitterFX V2, parameterName=SunDirectionLobe_Contrast}
  n46 = Power(
   Base=SmoothStep(Value=Multiply(A=@n44, B=@n45)) {constMin=-1}
   Exp=ScalarParameter() {defaultValue=4, group=GlitterFX V2, parameterName=SunDirectionLobeExp}
  )
  n47 = LinearInterpolate(
   A=ScalarParameter() {defaultValue=3, group=GlitterFX V2, parameterName=GlintExponent}
   B=ScalarParameter() {defaultValue=10, group=GlitterFX V2, parameterName=GlintExponentLit}
   Alpha=NamedRerouteDeclaration(None=Saturate(None=@n46)) {name=GlintsLitDirectionMask, nodeColor=<rgba 0 1 0.8125 1>}
  )
  n48 = NamedRerouteDeclaration(None=Multiply(A=Power(Base=@n39, Exp=@n47), B=NamedRerouteUsage())) {name=Glints MaskOutput, nodeColor=<rgba 1 0.9141 0 1>}
  n49 = LinearInterpolate(
   A=ScalarParameter() {defaultValue=10, group=GlitterFX V2, parameterName=GlintBrightness}
   B=ScalarParameter() {defaultValue=10, group=GlitterFX V2, parameterName=GlintBrightnessLit}
   Alpha=Saturate(None=NamedRerouteUsage())
  )
  NamedRerouteDeclaration(None=RayTracingQualitySwitch(Normal=Multiply(A=@n48, B=@n49) {constB=50}, RayTraced=@n48)) {name=Glints, nodeColor=<rgba 0.2891 1 0 1>}
  NamedRerouteDeclaration(None=@n6.Emissive Color) {name=Glint_InEmissive, nodeColor=<rgba 0 0.0884 0.8958 1>}
  NamedRerouteDeclaration(None=@n6.Normal) {name=Glint_InNormal, nodeColor=<rgba 0 0.0884 0.8958 1>}
  n50 = NamedRerouteUsage()
  n51 = Multiply(A=@n50, B=MaterialFunctionCall() {materialFunction=&Color_Time_of_Day}.Result)
  n52 = ScalarParameter() {group=GlitterFX V2, parameterName=Glints_TimeOfDayAdjustment}
  n53 = Constant() {r=0.5}
  n54 = Constant() {r=1}
  n55 = Constant() {r=0.5}
  n56 = Constant() {r=0.2}
  NamedRerouteDeclaration(None=Multiply(
    A=LinearInterpolate(A=@n50, B=@n51, Alpha=@n52)
    B=MaterialFunctionCall(Value_Morning=@n53, Value_Day=@n54, Value_Evening=@n55, Value_Night=@n56) {materialFunction=&Scalar_Time_of_Day}.Result
   )) {name=Glints_TODAdjusted, nodeColor=<rgba 0 0.5547 1 1>}
  NamedRerouteDeclaration(None=@n1.M) {name=Mask, nodeColor=<rgba 1 0 0.2109 1>}
  n57 = MaterialFunctionCall!(In=@n24.Result, TechArtMask=NamedRerouteUsage()) {materialFunction=/Game/Global/Materials/Character_Parent_2023/Fur_Parent_2023/MF_Fur_Shells}
  NamedRerouteDeclaration(None=MaterialFunctionCall!(Material=@n57.Result, Mask=NamedRerouteUsage(), Fur Length=@n57.Result) {materialFunction=/Game/Global/Materials/Character_Parent_2023/Fur_Parent_2023/MF_FurWPOShrink}.Result) {name=OUT_Material, nodeColor=<rgba 0 0.5312 1 1>}
  NamedRerouteDeclaration(None=ChannelMaskParameter(None=NamedRerouteDeclaration(None=@n1.Tech Art Mask) {name=TechArtMask, nodeColor=<rgba 0.2188 0 1 1>}) {group=_AirKringle, maskChannel=Blue, parameterName=SnowMaskChannel, defaultValue=<rgba 0 0 1 0>}.R) {name=Snow Mask, nodeColor=<rgba 0.5938 0 1 1>}
  NamedRerouteDeclaration(None=@n6.Base Color) {name=Glint_InBC, nodeColor=<rgba 0 0.0884 0.8958 1>}
  NamedRerouteDeclaration(None=@n26.A) {name=Glints_RawMask, nodeColor=<rgba 0.0312 0 1 1>}
  NamedRerouteDeclaration(None=@n6.Metallic) {name=Glint_InMetallic, nodeColor=<rgba 0 0.0884 0.8958 1>}
  NamedRerouteDeclaration(None=@n6.Roughness) {name=Glint_InRoughness, nodeColor=<rgba 0 0.0884 0.8958 1>}
  Reroute(None=@n6.Base Color)
  Reroute(None=@n6.Metallic)
  Reroute(None=@n6.Roughness)
  Reroute(None=@n6.Emissive Color)
  Reroute(None=@n6.Normal)
 ]
}
